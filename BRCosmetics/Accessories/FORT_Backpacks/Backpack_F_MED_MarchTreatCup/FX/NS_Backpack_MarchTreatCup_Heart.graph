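# Niagara effect for the MarchTreatCup back bling that emits short-lived pink heart sprites reacting to event audio.
# One emitter splits particles into two mirrored spawn groups that drift outward and up, drawn through four sprite renderers with separate color, size and glow attributes.
# Audio amplitude and System.MusicReactivity gate spawn probability, while amplitude also drives size and brightness.
# A front-end preview flag changes spawn and size response, and distant sprites scale up for readability.
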NiagaraSystem /BRCosmetics/Accessories/FORT_Backpacks/Backpack_F_MED_MarchTreatCup/FX/NS_Backpack_MarchTreatCup_Heart
dependencies {
 usedDataInterfaces [
  &NiagaraDataInterfaceCurve
  &NiagaraDataInterfaceArrayFloat
  &NiagaraDataInterfaceArrayFloat3
  &NiagaraDataInterfaceSpriteRendererInfo
 ]
 usedDynamicInputs []
 usedModules [
  &SystemState
  &self:SystemSpawnScript.SystemScriptSource.SystemScriptGraph.NiagaraNodeAssignment_0.SetVariables_#1
  &InterpolateOverTime
 ]
}
emitters {Heart001 {
  emitterUpdateScript [
   &EmitterState(Inactive Response=::NewEnumerator1, Life Cycle Mode=::NewEnumerator1, Loop Duration=1, UseLoopDelay=false)
   &SpawnRate(
    Spawn Probability=&MakeCustomFloatFromBool(
     Bool=$User.bIsFrontEndPreview
     True Float=&Lerp_Float(A=0.3, B=1, Alpha=$NPC.FortniteNPC.EventAudio_Amplitude)
     False Float=&MakeCustomFloatFromBool(
      Bool=&SetBoolByFloatComparison(A=$NPC.FortniteNPC.EventAudio_Amplitude, B=0, Comparison Type=::NewEnumerator4)
      True Float=&Lerp_Float(A=0, B=1, Alpha=$System.MusicReactivity)
      False Float=1
     )
    )
    SpawnRate=7
    Use Spawn Probability=true
   )
  ]
  particleSpawnScript [
   &FN_SetSpawnGroup(Number of Spawn Groups=2)
   &InitializeParticle(
    Lifetime Max=0.8
    Lifetime Min=0.4
    Lifetime Mode=::NewEnumerator1
    Sprite Rotation Angle=&SelectFloatFromArray(Direct Array Index=$Particles.SpawnGroup, Float Selection Array=DataInterface() {propertyValues {FloatData [-0.02,0.02]}})
    Sprite Rotation Mode=::NewEnumerator1
    Sprite Size Mode=::NewEnumerator1
    Uniform Sprite Size Max=16
    Uniform Sprite Size Min=13
    UsePositionOffset=false
    Write Material Random=true
   )
   &ShapeLocation(
    MaskSpawnGroup=false
    Non Uniform Scale=<1 1.3 1>
    Offset=&SelectVectorFromArray(Direct Array Index=$Particles.SpawnGroup, Vector Selection Array=DataInterface() {propertyValues {FloatData [<30 -20 10>, <-30 -20 10>]}})
    Offset Mode=::NewEnumerator0
    Override Seed=false
    Rotation Axis=<0 1 0>
    Rotation Mode=::NewEnumerator1
    Sphere Radius=20
    UseOwnerScale=true
   )
   &CameraOffset(Camera Offset Amount=20)
   &SubUVAnimation(
    End Frame Range Override=1
    "Sprite Renderer "=DataInterface() {propertyValues {SpriteRenderer=&self:Heart001_0.NiagaraSpriteRendererProperties_0}}
    SubUV Animation Mode:ENiagara_SubUVLookupModeV2=::NewEnumerator2
    UseEndFrame=true
    UseStartFrame=true
   )
   &SubUVAnimation(
    End Frame Range Override=1
    "Sprite Renderer "=DataInterface() {propertyValues {SpriteRenderer=&self:Heart001_0.NiagaraSpriteRendererProperties_1}}
    SubUV Animation Mode:ENiagara_SubUVLookupModeV2=::NewEnumerator2
    UseEndFrame=true
    UseStartFrame=true
   ) {module=SubUVAnimation001}
   &AddVelocity(Velocity=&SelectVectorFromArray(Direct Array Index=$Particles.SpawnGroup, Vector Selection Array=DataInterface() {propertyValues {FloatData [<10 0 8>, <-10 0 8>]}}))
  ]
  particleUpdateScript [
   &ParticleState()
   &SpriteRotationRate(Rotation Rate=&Multiply_Float(
     A=&RandomRangeFloat(Minimum=-10, Maximum=10)
     B=&FloatFromCurve(FloatCurve=DataInterface() {propertyValues {Curve {keys [<at 0 1>,<at 1 0>]}}}, Scale Curve=10)
    )) {enabled=false}
   &ScaleSpriteSize(
    Uniform Curve Scale=&Multiply_Float(
     A=&MakeCustomFloatFromBool(
      Bool=$User.bIsFrontEndPreview
      True Float=&RandomRangeFloat(Minimum=0.45, Maximum=0.75)
      False Float=&RandomRangeFloat(Minimum=0.5, Maximum=1)
     )
     B=&Multiply_Float(A=1, B=&ClampFloat(Float=$NPC.FortniteNPC.EventAudio_Amplitude, Min=0.6, Max=1.1))
    )
    Uniform Curve Sprite Scale=DataInterface() {propertyValues {Curve {keys [{tangentMode=RCTM_User, time=0, value=0.3}, <at 0.1 1.3>, <at 0.2 1>, <at 0.8 1>, {tangentMode=RCTM_User, time=1, value=0}]}}}
   )
   &ScaleSpriteSize(
    Uniform Curve Index=&FN_LOD_Distance_Remap(Start Distance=500, End Distance=1500, Minimum Output=0, Maximum Output=1, Clamp Output=true)
    Uniform Curve Sprite Scale=DataInterface() {propertyValues {Curve {keys [{tangentMode=RCTM_User, time=0.4, value=1}, {tangentMode=RCTM_User, time=0.6, value=2}]}}}
   ) {module=ScaleSpriteSize001}
   /SetVariables(Particles.Size002:Vector2f=&Multiply_Vector2D(A=$Particles.SpriteSize, B=&Vector2DFromFloat(Value=1.2))) {isSetParameters=true}
   &ScaleColor(
    Scale Alpha=&Multiply_Float(
     A=&FloatFromCurve(
      FloatCurve=DataInterface() {propertyValues {Curve {keys [{arriveTangent=-5.0651, interpMode=RCIM_Cubic, leaveTangent=-5.0651, tangentMode=RCTM_User, time=0.7, value=1}, <at 1 0 RCIM_Cubic>]}}}
      Scale Curve=3
     )
     B=1
    )
    Scale RGB=&Multiply_VectorByFloat(
     Vector=&MakeVectorFromLinearColorRGB(Color=&RandomRangeLinearColor(Minimum=<rgba 0.9323 0.0746 0.1995 1>, Maximum=<rgba 0.9323 0.1203 0.3276 1>))
     Float=&Lerp_Float(A=1, B=1.2, Alpha=$NPC.FortniteNPC.EventAudio_Amplitude)
    )
    ScaleA=true
    ScaleRGB=true
   )
   /SetVariables(Particles.Color002=&MakeLinearColorFromVectorAndFloat(
     "Vector (RGB)"=&MakeVectorFromLinearColorRGB(Color=<rgba 2 0.8438 1.11 1>)
     "Float (Alpha)"=&FloatFromCurve(FloatCurve=DataInterface() {propertyValues {Curve {keys [
          <at 0 1 RCIM_Cubic>
          {arriveTangent=-0.0129, interpMode=RCIM_Cubic, leaveTangent=-7.6717, tangentMode=RCTM_Break, time=0.7, value=1}
          <at 1 0 RCIM_Cubic>
         ]}}})
    )) {isSetParameters=true}
   /SetVariables(Particles.Color_glow:LinearColor=&MakeLinearColorFromVectorAndFloat(
     "Vector (RGB)"=&MakeVectorFromLinearColorRGB(Color=<rgba 1 0.3021 0.554 1>)
     "Float (Alpha)"=&FloatFromCurve(
      FloatCurve=DataInterface() {propertyValues {Curve {keys [
          <at 0 1 RCIM_Cubic>
          {arriveTangent=-0.0129, interpMode=RCIM_Cubic, leaveTangent=-7.6717, tangentMode=RCTM_Break, time=0.7, value=1}
          <at 1 0 RCIM_Cubic>
         ]}}}
      Scale Curve=0.8
     )
    )) {isSetParameters=true}
   &DynamicMaterialParameters(
    Index 0 Param 1=&FloatFromCurve(FloatCurve=DataInterface() {propertyValues {Curve {keys [<at 0 -1.5>,<at 0.7 -1.5>,<at 1 -1.5>]}}})
    Index 0 Param 2=0.2
    Index 1 Param 1=&FloatFromCurve(FloatCurve=DataInterface() {propertyValues {Curve {keys [<at 0 0 RCIM_Constant>,<at 0.107 0 RCIM_Constant>,<at 0.2819 0 RCIM_Constant>,<at 0.3721 1 RCIM_Constant>,<at 0.5142 1 RCIM_Constant>,<at 0.5496 0 RCIM_Constant>,<at 0.672 0 RCIM_Constant>,<at 0.7179 1 RCIM_Constant>,<at 0.8232 1 RCIM_Constant>,<at 0.8524 0 RCIM_Constant>,<at 1 0 RCIM_Constant>]}}})
    Module.Index0Param0Write=true
    Module.Index0Param1Write=true
    Module.Index0Param2Write=true
    Module.Index0Param3Write=true
    Module.Index1Param0Write=true
    Module.Index1Param1Write=true
    Module.Index1Param2Write=true
    Module.Index1Param3Write=true
    Write Parameter Index 1=true
   )
   &Drag(Use Linear Drag=true, Use Rotational Drag=true)
   &SolveForcesAndVelocity(Clamp Velocity=false, Limit Acceleration=false)
  ]
  rendererClasses [&NiagaraSpriteRendererProperties]
  renderers [
   {rendererClass=&NiagaraSpriteRendererProperties}
   {rendererClass=&NiagaraSpriteRendererProperties, rendererIndex=1}
   {rendererClass=&NiagaraSpriteRendererProperties, rendererIndex=2}
   {rendererClass=&NiagaraSpriteRendererProperties, rendererIndex=3}
  ]
 }}
summary {
 systemName=NS_Backpack_MarchTreatCup_Heart
 emitters [{emitterName=Heart001, rendererClasses [&NiagaraSpriteRendererProperties]}]
 userVariables [
  User.bIsFrontEnd {type=NiagaraBool}
  User.bIsFrontEndPreview {type=NiagaraBool}
 ]
}
