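# Control Rig for the MarchTreatCup backpack that procedurally poses its two wings from character movement states.
# One execution chain layers weighted additive global bone offsets for jumping, head rotation, skydiving, sliding, tactical sprint, crouching, swimming, parachuting and targeting.
# Spring or alpha interpolation smooths most weights; skydiving suppresses the jump pose, while movement and targeting suppress head rotation.
# Grinding, strafe and crouch-targeting branches exist but are disconnected from execution.
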ControlRigBlueprint /BRCosmetics/Accessories/FORT_Backpacks/Backpack_F_MED_MarchTreatCup/Meshes/F_MED_MarchTreatCup_Pack_CtrlRig
graphs [
 RigVMFunctionLibrary {links [], nodes []}
 RigVMModel {nodes [
   Subtract = &RigVMFunction_MathDoubleSub(B=RigVMVariableNode() {variable=GrindingSprintAlpha}.Value, A=RigVMVariableNode() {variable=GrindingAlpha}.Value)
   n1 = /Script/ControlRig.RigUnit_SpringInterpV2(
    CriticalDamping=2
    Current=0
    Force=0
    Strength=1
    TargetVelocityAmount=0
    bInitializeFromTarget=true
    bUseCurrentInput=False
    Target=@Subtract.Result
   ) {
    is=struct RigVMUnitNode
    template="SpringInterp::Execute(in Target,in Strength,in CriticalDamping,in Force,in bUseCurrentInput,in Current,in TargetVelocityAmount,in bInitializeFromTarget,out Result,out Velocity)"
   }
   &RigUnit_ModifyTransforms(
    ExecutePin=()
    ItemToModify="((Item=(Type=Bone,Name=\"wing_base_l\"),Transform=(Rotation=(X=0.2179,Y=0.0331,Z=-0.052,W=0.974),Translation=(X=0,Y=0,Z=0),Scale3D=(X=1,Y=1,Z=1))),(Item=(Type=Bone,Name=\"wing_base_r\"),Transform=(Rotation=(X=0,Y=0,Z=0,W=1),Translation=(X=0,Y=0,Z=0),Scale3D=(X ...(truncated)"
    Mode=AdditiveGlobal
    WeightMinimum=0
    Weight=@n1.Result
   )
   Multiply = &RigVMFunction_MathDoubleMul(B=RigVMVariableNode() {variable=GrindingSprintAlpha}.Value, A=RigVMVariableNode() {variable=GrindingAlpha}.Value)
   n2 = &RigUnit_SpringInterpV2(
    CriticalDamping=2
    Current=0
    Force=0
    Strength=1
    TargetVelocityAmount=0
    bInitializeFromTarget=true
    bUseCurrentInput=False
    Target=@Multiply.Result
   ) {template="SpringInterp::Execute(in Target,in Strength,in CriticalDamping,in Force,in bUseCurrentInput,in Current,in TargetVelocityAmount,in bInitializeFromTarget,out Result,out Velocity)"}
   > &RigUnit_ModifyTransforms(
    ItemToModify="((Item=(Type=Bone,Name=\"wing_base_l\"),Transform=(Rotation=(X=-0.0057,Y=-0.1304,Z=-0.0432,W=0.9905),Translation=(X=0,Y=0,Z=0),Scale3D=(X=1,Y=1,Z=1))),(Item=(Type=Bone,Name=\"wing_base_r\"),Transform=(Rotation=(X=0,Y=0,Z=0.3746,W=0.9272),Translation=(X=0,Y=0,Z ...(truncated)"
    Mode=AdditiveGlobal
    WeightMinimum=0
    Weight=@n2.Result
   )
   n3 = RigVMVariableNode() {variable=LeftRightSideMovement}
   n4 = &RigUnit_SpringInterpV2(
    CriticalDamping=0.9
    Current=0
    Force=0
    Strength=2
    TargetVelocityAmount=0
    bInitializeFromTarget=true
    bUseCurrentInput=False
    Target=@n3.Value
   ) {template="SpringInterp::Execute(in Target,in Strength,in CriticalDamping,in Force,in bUseCurrentInput,in Current,in TargetVelocityAmount,in bInitializeFromTarget,out Result,out Velocity)"}
   &RigUnit_ModifyTransforms(
    ExecutePin=()
    ItemToModify="((Item=(Type=Bone,Name=\"wing_base_l\"),Transform=(Rotation=(X=0,Y=0,Z=-0.0872,W=0.9962),Translation=(X=0,Y=0,Z=0),Scale3D=(X=1,Y=1,Z=1))),(Item=(Type=Bone,Name=\"wing_base_r\"),Transform=(Rotation=(X=0,Y=0,Z=-0.0872,W=0.9962),Translation=(X=0,Y=0,Z=0),Scale3D ...(truncated)"
    Mode=AdditiveGlobal
    WeightMinimum=0
    Weight=@n4.Result
   )
   n5 = &RigUnit_SpringInterpV2(
    CriticalDamping=0.9
    Current=0
    Force=0
    Strength=2
    TargetVelocityAmount=0
    bInitializeFromTarget=true
    bUseCurrentInput=False
    Target=/Script/RigVM.RigVMFunction_MathDoubleNegate(Value=@n3.Value) {is=struct RigVMUnitNode, template="Negate::Execute(in Value,out Result)"}.Result
   ) {template="SpringInterp::Execute(in Target,in Strength,in CriticalDamping,in Force,in bUseCurrentInput,in Current,in TargetVelocityAmount,in bInitializeFromTarget,out Result,out Velocity)"}
   > &RigUnit_ModifyTransforms(
    ItemToModify="((Item=(Type=Bone,Name=\"wing_base_l\"),Transform=(Rotation=(X=0,Y=0,Z=0.0872,W=0.9962),Translation=(X=0,Y=0,Z=0),Scale3D=(X=1,Y=1,Z=1))),(Item=(Type=Bone,Name=\"wing_base_r\"),Transform=(Rotation=(X=0,Y=0,Z=0.0872,W=0.9962),Translation=(X=0,Y=0,Z=0),Scale3D=( ...(truncated)"
    Mode=AdditiveGlobal
    WeightMinimum=0
    Weight=@n5.Result
   )
   RigUnit_BeginExecution = &RigUnit_BeginExecution()
   n6 = &RigVMFunction_MathDoubleAdd(
    A=&RigVMFunction_MathDoubleAdd(A=RigVMVariableNode() {variable=SkydiveDiveAlpha}.Value, B=RigVMVariableNode() {variable=SkydiveFloatingAlpha}.Value).Result
    B=RigVMVariableNode() {variable=SkydiveEnterAlpha}.Value
   )
   n7 = &RigVMFunction_AlphaInterp(
    InterpSpeedDecreasing=5
    InterpSpeedIncreasing=5
    Scale=1
    bClampResult=False
    bInterpResult=True
    Value=&RigVMFunction_MathDoubleMul(
     B=&RigVMFunction_MathDoubleRemap(SourceMaximum=1, SourceMinimum=0, TargetMaximum=0, TargetMinimum=1, Value=@n6.Result).Result
     A=RigVMVariableNode() {variable=JumpOrFalling}.Value
    ).Result
   )
   &RigUnit_ModifyTransforms(
    ItemToModify="((Item=(Type=Bone,Name=\"wing_base_l\"),Transform=(Rotation=(X=-0.0099,Y=-0.0866,Z=-0.1128,W=0.9898),Translation=(X=0,Y=0,Z=0),Scale3D=(X=1,Y=1,Z=1))),(Item=(Type=Bone,Name=\"wing_base_r\"),Transform=(Rotation=(X=-0.0099,Y=0.0866,Z=0.1128,W=0.9898),Translation ...(truncated)"
    Mode=AdditiveGlobal
    WeightMinimum=0
    ExecutePin=@RigUnit_BeginExecution.ExecutePin
    Weight=@n7.Result
   )
   Absolute = &RigVMFunction_MathDoubleAbs(Value=RigVMVariableNode() {variable=ForwardBackwardMovement}.Value)
   n8 = RigVMVariableNode() {variable=LeftRightSideMovement}
   n9 = &RigVMFunction_MathDoubleAdd(
    A=&RigVMFunction_MathDoubleAdd(A=@Absolute.Result, B=&RigVMFunction_MathDoubleAbs(Value=@n8.Value).Result).Result
    B=RigVMVariableNode() {variable=isTargeting}.Value
   )
   n10 = &RigVMFunction_AlphaInterp(
    InterpSpeedDecreasing=7.5
    InterpSpeedIncreasing=15
    Scale=1
    bClampResult=False
    bInterpResult=True
    Value=&RigVMFunction_MathDoubleMul(
     B=&RigVMFunction_MathDoubleRemap(SourceMaximum=1, SourceMinimum=0, TargetMaximum=0, TargetMinimum=1, Value=@n9.Result).Result
     A=RigVMVariableNode() {variable=HeadRotation}.Value
    ).Result
   )
   > &RigUnit_ModifyTransforms(
    ItemToModify="((Item=(Type=Bone,Name=\"wing_base_l\"),Transform=(Rotation=(X=0,Y=0,Z=0,W=1),Translation=(X=0,Y=0,Z=0),Scale3D=(X=1,Y=1,Z=1))),(Item=(Type=Bone,Name=\"wing_base_r\"),Transform=(Rotation=(X=0,Y=0,Z=0.3746,W=0.9272),Translation=(X=0,Y=0,Z=0),Scale3D=(X=1,Y=1,Z= ...(truncated)"
    Mode=AdditiveGlobal
    WeightMinimum=0
    Weight=@n10.Result
   )
   n11 = &RigUnit_SpringInterpV2(
    CriticalDamping=1.5
    Current=0
    Force=0
    Strength=2
    TargetVelocityAmount=0
    bInitializeFromTarget=true
    bUseCurrentInput=False
    Target=RigVMVariableNode() {variable=SkydiveFloatingAlpha}.Value
   ) {template="SpringInterp::Execute(in Target,in Strength,in CriticalDamping,in Force,in bUseCurrentInput,in Current,in TargetVelocityAmount,in bInitializeFromTarget,out Result,out Velocity)"}
   > &RigUnit_ModifyTransforms(
    ItemToModify="((Item=(Type=Bone,Name=\"wing_base_l\"),Transform=(Rotation=(X=0.0871,Y=0.0023,Z=0.0261,W=0.9959),Translation=(X=1,Y=-1,Z=0),Scale3D=(X=1,Y=1,Z=1))),(Item=(Type=Bone,Name=\"wing_base_r\"),Transform=(Rotation=(X=0.0871,Y=-0.0023,Z=-0.0261,W=0.9959),Translation= ...(truncated)"
    Mode=AdditiveGlobal
    WeightMinimum=0
    Weight=@n11.Result
   )
   Add = &RigVMFunction_MathDoubleAdd(A=RigVMVariableNode() {variable=SkydiveDiveAlpha}.Value, B=RigVMVariableNode() {variable=SkydiveEnterAlpha}.Value)
   Clamp = &RigVMFunction_MathDoubleClamp(Value=@Add.Result)
   n12 = &RigUnit_SpringInterpV2(
    CriticalDamping=0.85
    Current=0
    Force=0
    Strength=2
    TargetVelocityAmount=0
    bInitializeFromTarget=true
    bUseCurrentInput=False
    Target=@Clamp.Result
   ) {template="SpringInterp::Execute(in Target,in Strength,in CriticalDamping,in Force,in bUseCurrentInput,in Current,in TargetVelocityAmount,in bInitializeFromTarget,out Result,out Velocity)"}
   > &RigUnit_ModifyTransforms(
    ItemToModify="((Item=(Type=Bone,Name=\"wing_base_l\"),Transform=(Rotation=(X=0.0847,Y=-0.0203,Z=-0.2326,W=0.9687),Translation=(X=0,Y=0,Z=0),Scale3D=(X=1,Y=1,Z=1))),(Item=(Type=Bone,Name=\"wing_base_r\"),Transform=(Rotation=(X=0.0847,Y=0.0203,Z=0.2326,W=0.9687),Translation=( ...(truncated)"
    Mode=AdditiveGlobal
    WeightMinimum=0
    Weight=@n12.Result
   )
   n13 = &RigUnit_SpringInterpV2(
    CriticalDamping=0.9
    Current=0
    Force=0
    Strength=2
    TargetVelocityAmount=0
    bInitializeFromTarget=true
    bUseCurrentInput=False
    Target=RigVMVariableNode() {variable=SlidingAlpha}.Value
   ) {template="SpringInterp::Execute(in Target,in Strength,in CriticalDamping,in Force,in bUseCurrentInput,in Current,in TargetVelocityAmount,in bInitializeFromTarget,out Result,out Velocity)"}
   > &RigUnit_ModifyTransforms(
    ItemToModify="((Item=(Type=Bone,Name=\"wing_base_l\"),Transform=(Rotation=(X=0.0426,Y=0.0094,Z=0.2162,W=0.9754),Translation=(X=0,Y=0,Z=0),Scale3D=(X=1,Y=1,Z=1))),(Item=(Type=Bone,Name=\"wing_base_r\"),Transform=(Rotation=(X=0.0426,Y=-0.0094,Z=-0.2162,W=0.9754),Translation=( ...(truncated)"
    Mode=AdditiveGlobal
    WeightMinimum=0
    Weight=@n13.Result
   )
   n14 = &RigUnit_SpringInterpV2(
    CriticalDamping=0.9
    Current=0
    Force=0
    Strength=2
    TargetVelocityAmount=0
    bInitializeFromTarget=true
    bUseCurrentInput=False
    Target=RigVMVariableNode() {variable=TacticalSprintingAlpha}.Value
   ) {template="SpringInterp::Execute(in Target,in Strength,in CriticalDamping,in Force,in bUseCurrentInput,in Current,in TargetVelocityAmount,in bInitializeFromTarget,out Result,out Velocity)"}
   > &RigUnit_ModifyTransforms(
    ItemToModify="((Item=(Type=Bone,Name=\"wing_base_l\"),Transform=(Rotation=(X=0.0858,Y=-0.0151,Z=-0.173,W=0.9811),Translation=(X=0,Y=0,Z=0),Scale3D=(X=1,Y=1,Z=1))),(Item=(Type=Bone,Name=\"wing_base_r\"),Transform=(Rotation=(X=0.0858,Y=0.0151,Z=0.173,W=0.9811),Translation=(X= ...(truncated)"
    Mode=AdditiveGlobal
    WeightMinimum=0
    Weight=@n14.Result
   )
   n15 = &RigUnit_SpringInterpV2(
    CriticalDamping=0.9
    Current=0
    Force=0
    Strength=2
    TargetVelocityAmount=0
    bInitializeFromTarget=true
    bUseCurrentInput=False
    Target=RigVMVariableNode() {variable=isCrouching}.Value
   ) {template="SpringInterp::Execute(in Target,in Strength,in CriticalDamping,in Force,in bUseCurrentInput,in Current,in TargetVelocityAmount,in bInitializeFromTarget,out Result,out Velocity)"}
   > &RigUnit_ModifyTransforms(
    ItemToModify="((Item=(Type=Bone,Name=\"wing_base_l\"),Transform=(Rotation=(X=0,Y=0,Z=-0.1736,W=0.9848),Translation=(X=0,Y=0,Z=0),Scale3D=(X=1,Y=1,Z=1))),(Item=(Type=Bone,Name=\"wing_base_r\"),Transform=(Rotation=(X=0,Y=0,Z=0.1564,W=0.9877),Translation=(X=0,Y=0,Z=0),Scale3D= ...(truncated)"
    Mode=AdditiveGlobal
    WeightMinimum=0
    Weight=@n15.Result
   )
   n16 = &RigVMFunction_AlphaInterp(
    InterpSpeedDecreasing=200
    InterpSpeedIncreasing=200
    Scale=1
    bClampResult=False
    bInterpResult=True
    Value=RigVMVariableNode() {variable=SpineSwimmingAlpha}.Value
   )
   > &RigUnit_ModifyTransforms(
    ItemToModify="((Item=(Type=Bone,Name=\"wing_base_l\"),Transform=(Rotation=(X=-0.1565,Y=0.1472,Z=0.1919,W=0.9576),Translation=(X=0,Y=0,Z=0),Scale3D=(X=1,Y=1,Z=1))),(Item=(Type=Bone,Name=\"wing_base_r\"),Transform=(Rotation=(X=-0.1565,Y=-0.1472,Z=-0.1919,W=0.9576),Translation ...(truncated)"
    Mode=AdditiveGlobal
    WeightMinimum=0
    Weight=@n16.Result
   )
   > &RigUnit_ModifyTransforms(
    ItemToModify="((Item=(Type=Bone,Name=\"wing_base_l\"),Transform=(Rotation=(X=0.0868,Y=-0.0076,Z=-0.0868,W=0.9924),Translation=(X=0,Y=0,Z=0),Scale3D=(X=1,Y=1,Z=1))),(Item=(Type=Bone,Name=\"wing_base_r\"),Transform=(Rotation=(X=0.0868,Y=0.0076,Z=0.0868,W=0.9924),Translation=( ...(truncated)"
    Mode=AdditiveGlobal
    WeightMinimum=0
    Weight=RigVMVariableNode() {variable=ParachuteMoving}.Value
   )
   n17 = &RigUnit_SpringInterpV2(
    CriticalDamping=0.9
    Current=0
    Force=0
    Strength=2
    TargetVelocityAmount=0
    bInitializeFromTarget=true
    bUseCurrentInput=False
    Target=&RigVMFunction_MathDoubleMul(B=1, A=RigVMVariableNode() {variable=isTargeting}.Value).Result
   ) {template="SpringInterp::Execute(in Target,in Strength,in CriticalDamping,in Force,in bUseCurrentInput,in Current,in TargetVelocityAmount,in bInitializeFromTarget,out Result,out Velocity)"}
   > &RigUnit_ModifyTransforms(
    ItemToModify=((Item=(Type=Bone,Name="None"),Transform=(Rotation=(X=0,Y=0,Z=0,W=1),Translation=(X=0,Y=0,Z=0),Scale3D=(X=1,Y=1,Z=1))),(Item=(Type=Bone,Name="wing_base_r"),Transform=(Rotation=(X=0,Y=0,Z=0.4226,W=0.9063),Translation=(X=0,Y=0,Z=0),Scale3D=(X=1,Y=1,Z=1))))
    Mode=AdditiveGlobal
    WeightMinimum=0
    Weight=@n17.Result
   )
   n18 = RigVMVariableNode() {variable=isCrouching}
   n19 = &RigUnit_SpringInterpV2(
    CriticalDamping=0.9
    Current=0
    Force=0
    Strength=2
    TargetVelocityAmount=0
    bInitializeFromTarget=true
    bUseCurrentInput=False
    Target=&RigVMFunction_MathDoubleMul(B=&RigVMFunction_MathDoubleMul(B=0.25, A=@n18.Value).Result, A=RigVMVariableNode() {variable=isTargeting}.Value).Result
   ) {template="SpringInterp::Execute(in Target,in Strength,in CriticalDamping,in Force,in bUseCurrentInput,in Current,in TargetVelocityAmount,in bInitializeFromTarget,out Result,out Velocity)"}
   &RigUnit_ModifyTransforms(
    ExecutePin=()
    ItemToModify="((Item=(Type=Bone,Name=\"wing_base_l\"),Transform=(Rotation=(X=0,Y=0,Z=0.1736,W=0.9848),Translation=(X=0,Y=0,Z=0),Scale3D=(X=1,Y=1,Z=1))),(Item=(Type=Bone,Name=\"wing_base_r\"),Transform=(Rotation=(X=0,Y=0,Z=-0.2588,W=0.9659),Translation=(X=0,Y=0,Z=0),Scale3D= ...(truncated)"
    Mode=AdditiveGlobal
    WeightMinimum=0
    Weight=@n19.Result
   )
   RigVMVariableNode() {variable=Jethro_Crusty}
  ]}
]
hierarchy [
 root {loc=<0 0 0>, rot=<0 0 0 1>}
 pack {parent=root, loc=<0 -0.7316 0>, rot=<0 0 0 1>}
 wing_base_l {parent=pack, loc=<1.052 -0.9465 -4.1006>, rot=<0.1932 0.6802 0.1932 -0.6802>}
 wing_mane_l {parent=wing_base_l, loc=<8.2618 0.0217 -4.8178>, rot=<0.0579 -0.5764 0.0162 -0.815>}
 wing_01_l {parent=wing_mane_l, loc=<0 0 0>, rot=<0 0 0 1>}
 wing_02_l {parent=wing_01_l, loc=<21 0 0>, rot=<-0.0022 0.1904 0.0287 -0.9813>}
 wing_e_l {parent=wing_02_l, loc=<12 0 0>, rot=<0 0 0.0325 -0.9995>}
 wing_d_l {parent=wing_02_l, loc=<6.4705 0.3602 -5.8428>, rot=<-0.0118 -0.3615 0.0303 -0.9318>}
 wing_b_l {parent=wing_01_l, loc=<11.4919 0 -9.2738>, rot=<0 -0.5694 0 -0.8221>}
 wing_c_l {parent=wing_01_l, loc=<18.151 0 -7.472>, rot=<0 -0.381 0 -0.9246>}
 wing_a_l {parent=wing_01_l, loc=<2.1616 0 -8.138>, rot=<0 -0.6542 0 -0.7563>}
 wing_base_r {parent=pack, loc=<-1.052 -0.9465 -4.1006>, rot=<0.1932 -0.6802 -0.1932 -0.6802>}
 wing_mane_r {parent=wing_base_r, loc=<-8.2618 0.0217 -4.8178>, rot=<0.815 0.0162 0.5764 0.0579>}
 wing_01_r {parent=wing_mane_r, loc=<0 0 0>, rot=<0 0 0 1>}
 wing_02_r {parent=wing_01_r, loc=<-21 0 0>, rot=<-0.0022 0.1904 0.0287 -0.9813>}
 wing_e_r {parent=wing_02_r, loc=<-12.0001 0 0>, rot=<0 0 0.0325 -0.9995>}
 wing_d_r {parent=wing_02_r, loc=<-6.4704 -0.3602 5.8427>, rot=<-0.0118 -0.3615 0.0303 -0.9318>}
 wing_b_r {parent=wing_01_r, loc=<-11.4919 0 9.2738>, rot=<0 -0.5694 0 -0.8221>}
 wing_c_r {parent=wing_01_r, loc=<-18.151 0 7.472>, rot=<0 -0.381 0 -0.9246>}
 wing_a_r {parent=wing_01_r, loc=<-2.1616 0 8.138>, rot=<0 -0.6542 0 -0.7563>}
]
variables [
 ChildRotationInfluence {type=double, default=0.25}
 ForwardBackwardMovement {type=double, default=0}
 GrindingAlpha {type=double, default=0}
 GrindingSprintAlpha {type=double, default=0}
 HeadRotation {type=double, default=0}
 Jethro_Crusty {type=double, default=0}
 JumpOrFalling {type=double, default=0}
 LeftRightSideMovement {type=double, default=0}
 ParachuteMoving {type=double, default=0}
 SelfRotationInfluence {type=double, default=0.75}
 SkydiveDiveAlpha {type=double, default=0}
 SkydiveEnterAlpha {type=double, default=0}
 SkydiveFloatingAlpha {type=double, default=0}
 SlidingAlpha {type=double, default=0}
 SpineSwimmingAlpha {type=double, default=0}
 TacticalSprintingAlpha {type=double, default=0}
 isCrouching {type=double, default=0}
 isTargeting {type=double, default=0}
]
