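# Physics asset for the MarchTreatCup backpack cosmetic, giving its pack and pair of mirrored wings capsule bodies for physical motion.
# Each side chains wing_01 to the pack and wing_02 to wing_01, with segments a, b and c branching from wing_01 and d and e from wing_02.
# Constraint limits allow modest swing and twist along the main chain, while the branch segments have a zero second swing limit, favoring single-plane flapping.
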PhysicsAsset /BRCosmetics/Accessories/FORT_Backpacks/Backpack_F_MED_MarchTreatCup/Meshes/F_MED_MarchTreatCup_Pack_Physics
bodies {
 pack [{shapeType=Capsule, length=2.6834, radius=3.5813, center=<0 0.3683 -4.1796>, rotation=<rot -0.0004 -90 0>}]
 wing_01_l [{shapeType=Capsule, length=10.4979, radius=3.9388, shapeName=wing_01_l_capsule, center=<8.6043 0 0>, rotation=<rot 90 -179.9998 0>}]
 wing_01_r [{shapeType=Capsule, length=10.4979, radius=3.9388, shapeName=wing_01_r_capsule, center=<-8.6043 0 0>, rotation=<rot 90 0.0002 0>}]
 wing_02_l [{shapeType=Capsule, length=5.0728, radius=3.1451, shapeName=wing_02_l_capsule, center=<5.8381 0 0>, rotation=<rot -90 179.9997 0>}]
 wing_02_r [{shapeType=Capsule, length=5.0728, radius=3.1451, shapeName=wing_02_r_capsule, center=<-5.8381 0 0>, rotation=<rot -90 -0.0003 0>}]
 wing_a_l [{shapeType=Capsule, length=4.9992, radius=3.1517, shapeName=wing_a_l_capsule, center=<5.5568 0 0>, rotation=<rot -90 180 0>}]
 wing_a_r [{shapeType=Capsule, length=4.9992, radius=3.1517, shapeName=wing_a_r_capsule, center=<-5.5568 0 0>, rotation=<rot -90 0 0>}]
 wing_b_l [{shapeType=Capsule, length=6, radius=3, shapeName=wing_b_l_capsule, center=<5.9575 0.0001 0>, rotation=<rot 90 -179.9989 0>}]
 wing_b_r [{shapeType=Capsule, length=6, radius=3, shapeName=wing_b_r_capsule, center=<-5.9575 -0.0001 0>, rotation=<rot 90 0.0011 0>}]
 wing_c_l [{shapeType=Capsule, length=6, radius=3, shapeName=wing_c_l_capsule, center=<5.9575 0.0001 0>, rotation=<rot 90 -179.9989 0>}]
 wing_c_r [{shapeType=Capsule, length=6, radius=3, shapeName=wing_c_r_capsule, center=<-5.9575 -0.0001 0>, rotation=<rot 90 0.0011 0>}]
 wing_d_l [{shapeType=Capsule, length=6, radius=3, shapeName=wing_d_l_capsule, center=<5.9575 0.0001 0>, rotation=<rot 90 -179.9993 0>}]
 wing_d_r [{shapeType=Capsule, length=6, radius=3, shapeName=wing_d_r_capsule, center=<-5.9575 -0.0001 0>, rotation=<rot 90 0.0007 0>}]
 wing_e_l [{shapeType=Capsule, length=6, radius=3, shapeName=wing_e_l_capsule, center=<5.9575 0 0>, rotation=<rot 90 -180 0>}]
 wing_e_r [{shapeType=Capsule, length=6, radius=3, shapeName=wing_e_r_capsule, center=<-5.9575 0 0>, rotation=<rot 90 0 0>}]
}
constraints [
 {bone1Name=wing_01_r, bone2Name=pack, swing1LimitDegrees=15, swing2LimitDegrees=30, twistLimitDegrees=10}
 {bone1Name=wing_02_r, bone2Name=wing_01_r, swing1LimitDegrees=10, swing2LimitDegrees=10, twistLimitDegrees=5}
 {bone1Name=wing_01_l, bone2Name=pack, swing1LimitDegrees=15, swing2LimitDegrees=30, twistLimitDegrees=10}
 {bone1Name=wing_02_l, bone2Name=wing_01_l, swing1LimitDegrees=10, swing2LimitDegrees=10, twistLimitDegrees=5}
 {bone1Name=wing_c_r, bone2Name=wing_01_r, swing1LimitDegrees=30, swing2LimitDegrees=0, twistLimitDegrees=10}
 {bone1Name=wing_d_r, bone2Name=wing_02_r, swing1LimitDegrees=15, swing2LimitDegrees=0, twistLimitDegrees=10}
 {bone1Name=wing_e_r, bone2Name=wing_02_r, swing1LimitDegrees=15, swing2LimitDegrees=0, twistLimitDegrees=10}
 {bone1Name=wing_a_r, bone2Name=wing_01_r, swing1LimitDegrees=30, swing2LimitDegrees=0, twistLimitDegrees=10}
 {bone1Name=wing_b_r, bone2Name=wing_01_r, swing1LimitDegrees=30, swing2LimitDegrees=0, twistLimitDegrees=10}
 {bone1Name=wing_c_l, bone2Name=wing_01_l, swing1LimitDegrees=30, swing2LimitDegrees=0, twistLimitDegrees=10}
 {bone1Name=wing_d_l, bone2Name=wing_02_l, swing1LimitDegrees=15, swing2LimitDegrees=0, twistLimitDegrees=10}
 {bone1Name=wing_e_l, bone2Name=wing_02_l, swing1LimitDegrees=15, swing2LimitDegrees=0, twistLimitDegrees=10}
 {bone1Name=wing_a_l, bone2Name=wing_01_l, swing1LimitDegrees=30, swing2LimitDegrees=0, twistLimitDegrees=10}
 {bone1Name=wing_b_l, bone2Name=wing_01_l, swing1LimitDegrees=30, swing2LimitDegrees=0, twistLimitDegrees=10}
]
