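# Animation Blueprint for the MarchTreatCup back bling, driving the pack's secondary motion from physics and player state.
# The input pose feeds two RigidBody sims, full and partial strength, selected by front-end state, then an additive offset on the pack bone from spine bend and a Control Rig.
# The update event converts movement, swimming, turn-in-place and parachute states into rig alphas.
# Skydive-entry and targeting events set others, with debug prints left in.
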AnimBlueprint /BRCosmetics/Accessories/FORT_Backpacks/Backpack_F_MED_MarchTreatCup/Meshes/F_MED_MarchTreatCup_Pack_AnimBP
target_skeleton=&/F_MED_MarchTreatCup_Pack_Skeleton
graphs [
 AnimGraph {nodes [
   n1 = RigidBody(ComponentPose=Local To Component(LocalPose=Use cached pose 'Main pose'()), Alpha=0.6) {
    bOverrideWorldGravity=true
    overridePhysicsAsset=&/F_MED_MarchTreatCup_Pack_Physics
    alphaBoolBlend {blendOption=Linear}
    baseBoneRef {boneName=pack}
    overrideWorldGravity {z=-980}
    simSpaceSettings {maxAngularAcceleration=10, maxAngularVelocity=3, maxLinearAcceleration=10, maxLinearVelocity=10, velocityScaleZ=0.2, worldAlpha=1}
   }
   n2 = Component To Local(ComponentPose=@n1)
   n3 = RigidBody(ComponentPose=Local To Component(LocalPose=Use cached pose 'Main pose'())) {
    bOverrideWorldGravity=true
    overridePhysicsAsset=&F_MED_MarchTreatCup_Pack_Physics
    alphaBoolBlend {blendOption=Linear}
    baseBoneRef {boneName=pack}
    overrideWorldGravity {z=-980}
    simSpaceSettings {maxAngularAcceleration=10, maxAngularVelocity=3, maxLinearAcceleration=10, maxLinearVelocity=10, velocityScaleZ=0.2, worldAlpha=1}
   }
   n4 = Component To Local(ComponentPose=@n3)
   n5 = Get bIsInFrontEnd()
   n6 = "Transform (Modify) Bone - Bone: pack"(
    ComponentPose=Local To Component(LocalPose=Blend Poses by bool(BlendPose_0=@n2, BlendPose_1=@n4, bActiveValue=@n5))
    Translation=Get backpack_translation()
    Rotation=Get backpack_rotate()
   ) {rotationMode=BMM_Additive, translationMode=BMM_Additive, alphaBoolBlend {blendOption=Linear}, boneToModify {boneName=pack}}
   n7 = Control Rig!(
    Source=Component To Local(ComponentPose=@n6)
    Alpha
    SkydiveDiveAlpha:double=Get SkydiveDiveModeAlpha() -> (SkydiveDiveModeAlpha:double)
    SkydiveEnterAlpha:double=Get SkydiveEnterAlpha() -> (SkydiveEnterAlpha:double)
    TacticalSprintingAlpha:double=Get TacticalSprintingAlpha() -> (TacticalSprintingAlpha:double)
    GrindingSprintAlpha:double=Get GrindingSprintAlpha() -> (GrindingSprintAlpha:double)
    GrindingAlpha:double=Get GrindingAlpha() -> (GrindingAlpha:double)
    SlidingAlpha:double=Get SlidingAlpha() -> (SlidingAlpha:double)
    SkydiveFloatingAlpha:double=Get SkydiveFloatingAlpha() -> (SkydiveFloatingAlpha:double)
    JumpOrFalling:double=Get JumpOrFallingAlpha() -> (JumpOrFallingAlpha:double)
    isCrouching:double=Get CrouchingAlpha() -> (CrouchingAlpha:double)
    isTargeting:double=Get TargetingAlpha() -> (TargetingAlpha:double)
    LeftRightSideMovement:double=Get Player Vector LR() -> (Player Vector LR:double)
    ForwardBackwardMovement:double=Get Player Vector Fwd Bwd() -> (Player Vector Fwd Bwd:double)
    SpineSwimmingAlpha:double=Get SpineSwimmingAlpha() -> (SpineSwimmingAlpha:float)
    HeadRotation:double=Get TurnInTwistAlpha() -> (TurnInTwistAlpha:float)
    ParachuteMoving:double=Get ParachuteMoving() -> (ParachuteMoving:double)
    Jethro_Crusty:double=0
   ) -> (Pose) {
    alphaBoolBlend {blendOption=Linear}
    controlRigAssetReference {blueprintRigClass=&/F_MED_MarchTreatCup_Pack_CtrlRig.F_MED_MarchTreatCup_Pack_CtrlRig_C}
   }
   Output Pose(Result=@n7) {layerGroup=None}
   Save cached pose 'Main pose'(Pose=Input Pose()) {cacheName=Main pose}
  ]}
 EventGraph {nodes [
   Event BlueprintUpdateAnimation()
   > n1 = Sequence()
   > Set backpack_rotate(backpack_rotate="0, 0, 0")
   n2 = GetOwnerSkeletalMeshComponent()
   n3 = BreakRotator(InRot=BreakTransform(InTransform=GetDeltaTransformFromRefPose(self=@n2, BoneName=spine_05, BaseName=spine_01)).Rotation)
   > Set backpack_translation(backpack_translation=MakeVector(Y=0, Z=MapRangeClamped(Value=@n3.Yaw, InRangeA=-20, InRangeB=-50, OutRangeB=4)))
   > n4 = Branch(Condition=Get bIsInFrontEnd())
   > Set SimlationSpace(SimlationSpace:AnimPhysSimSpaceType Enum=Component) -> (Output_Get:AnimPhysSimSpaceType Enum)
   Set SimlationSpace(execute=@n4.else)
   n5 = PlayerMovement(execute=@n1.then_1, AnimInstance=Self-Reference(), XDotProductCoefficient=1, YDotProductCoefficient=1, ZDotProductCoefficient=1)
   > Set Player Vector LR(Player Vector LR:double=@n5) -> (Output_Get:double)
   > Set Player Vector Fwd Bwd(Player Vector Fwd Bwd:double=@n5.YMovement) -> (Output_Get:double)
   > Set SkydiveDiveModeAlpha(SkydiveDiveModeAlpha:double="To Float (Boolean)"(InBool=Get bIsSkydiveDiveMode())) -> (Output_Get:double)
   > Set bIsMaleCrouching(bIsMaleCrouching:bool=AND Boolean(A=Get bIsMale() -> (bIsMale:bool), B=Get bIsCrouching())) -> (Output_Get:bool)
   n6 = AND Boolean(A=NOT Boolean(A=Get bIsCrouching()), B=NOT Boolean(A=Get bIsSliding()))
   n7 = AND Boolean!(A=Get bIsTacticalSprinting(), B=NOT Boolean(A=Get bIsJumping()), C:bool=@n6) -> (ReturnValue)
   n8 = "To Float (Boolean)"(InBool=AND Boolean!(A=Get bIsSprinting(), B=@n6) -> (ReturnValue))
   > Set TacticalSprintingAlpha(TacticalSprintingAlpha:double=float + float(A="To Float (Boolean)"(InBool=@n7), B=float * float(A=@n8, B=0.5) {is=title K2Node_PromotableOperator})) -> (Output_Get:double)
   > Set GrindingAlpha(GrindingAlpha:double="To Float (Boolean)"(InBool=Get bIsOnGrindRail())) -> (Output_Get:double)
   > Set SlidingAlpha(SlidingAlpha:double="To Float (Boolean)"(InBool=Get bIsSliding())) -> (Output_Get:double)
   > Set SkydiveFloatingAlpha(SkydiveFloatingAlpha:double="To Float (Boolean)"(InBool=Get bIsSkydiveFloating())) -> (Output_Get:double)
   n9 = NOT Boolean(A=Get bIsTacticalSprinting())
   n10 = NOT Boolean(A=Get bIsParachuteOpened())
   n11 = OR Boolean(A=Get bIsJumping(), B=Get bIsFalling())
   > Set JumpOrFallingAlpha(JumpOrFallingAlpha:double="To Float (Boolean)"(InBool=AND Boolean!(A=@n11, B=@n10, C:bool=@n9) -> (ReturnValue))) -> (Output_Get:double)
   n12 = NOT Boolean(A=Get bIsSliding())
   n13 = Get bIsCrouching()
   > Set CrouchingAlpha(CrouchingAlpha:double="To Float (Boolean)"(InBool=AND Boolean!(A=@n13, B=@n12) -> (ReturnValue))) -> (Output_Get:double)
   n14 = GetOwnerSkeletalMeshComponent()
   n15 = Get bIsSurfaceSwimming()
   > Set SpineSwimmingAlpha(SpineSwimmingAlpha:float=float * float(
     A=GetAlphaFromReferenceTransformDelta(SkeletalMeshComponent=@n14, BoneName=spine_05, BaseName=spine_01, InputRange=(X=-30,Y=0), OutputRange=(X=1,Y=0), TwistAxis="1,0,0")
     B="To Float (Boolean)"(InBool=@n15)
    )) -> (Output_Get:float)
   n16 = GetTurnInPlaceInput!(self:Fort Player Animation Interface Object Reference=GetOwnerAnimationInterface!(self:Custom Character Part Anim Instance Object Reference) -> (ReturnValue:Fort Player Animation Interface Object Reference)) -> (ReturnValue:"Fort Anim Input Turn In Place Structure (by ref)")
   > Set TurnInTwistAlpha(TurnInTwistAlpha:float=Remap Float(
     Input=Break Fort Anim Input Turn in Place!(FortAnimInput_TurnInPlace:"Fort Anim Input Turn In Place Structure (by ref)"=@n16) -> (
      RootYawOffset:float=~
      TurnAnimPlayRate:float=~
      bWantsToTurn:bool=~
      bExitTurn:bool=~
      bUseTurnA:bool=~
      TurnACurveAlpha:float=~
      TurnBCurveAlpha:float=~
      bWantsToTurnInVehicle:bool=~
      bWantsToTurnAgain:bool=~
      bTurningLeft:bool=~
      ShufflePlayRate:float=~
      CrouchShufflePlayRate:float=~
      CrouchShuffleStartPosition:float=~
      CrouchShuffleSpeedScaling:float=~
      bTransition_LowerBody_Shuffle:bool=~
      bTransition_LowerBody_ShuffleADS:bool=~
      bPlayOutro:bool=~
     ) {is=title K2Node_BreakStruct}
     OldRange=(X=-35,Y=-75)
     NewRange=(X=0,Y=1)
     Clamp=true
    ) -> (Output=~)) -> (Output_Get:float)
   > Set ParachuteMoving(ParachuteMoving:double="To Float (Boolean)"(InBool=Get bIsParachuteOpened())) -> (Output_Get:double)
   Event OnBeginSkydiving()
   > Set SkydiveEnterAlpha(SkydiveEnterAlpha:double=1) -> (Output_Get:double)
   > Delay(Duration=1)
   > Set SkydiveEnterAlpha(SkydiveEnterAlpha=0)
   Event OnBeginTargeting() {is=title K2Node_Event}
   > Set TargetingAlpha(TargetingAlpha:double=1) -> (Output_Get:double)
   > PrintString!(execute, InString=Hello 1, bPrintToScreen, bPrintToLog, TextColor, Duration, Key) -> (then)
   Event OnEndTargeting() {is=title K2Node_Event}
   > Set TargetingAlpha(TargetingAlpha=0)
   > PrintString(InString=Hello 2)
   n17 = TryGetFortPawn(self=GetOwnerFortPlayerAnimInstance!(self:Custom Character Part Anim Instance Object Reference) -> (ReturnValue:Fort Player Anim Instance Object Reference))
   n18 = Contains(
    SearchIn=GetDisplayName(Object=GetCurrentWeapon!(self:Fort Pawn Object Reference=@n17) -> (ReturnValue:Fort Weapon Object Reference))
    Substring=Jethro_Crusty
   )
   Set Jethro_Crusty(Jethro_Crusty:double="To Float (Boolean)"(InBool=@n18)) -> (Output_Get:double)
   > PrintString(InString="To String (Boolean)"(InBool=@n18))
   TryGetPawnOwner()
   MakeVector(Y=0)
   MapRangeClamped(Value=@n3, InRangeA=0, InRangeB=-50, OutRangeB=4)
   "To Float (Boolean)"(InBool=Get bIsTargeting())
   AND Boolean(A=false, B=Get bIsMoving2D())
  ]}
]
variables [
 backpack_rotate {type=Rotator}
 backpack_translation {type=Vector}
 bIsMale {type=boolean}
 bIsMaleCrouching {type=boolean}
 CrouchingAlpha {type=number}
 GrindingAlpha {type=number}
 GrindingSprintAlpha {type=number}
 Jethro_Crusty {type=number}
 JumpOrFallingAlpha {type=number}
 ParachuteMoving {type=number}
 Player Vector Fwd Bwd {type=number}
 Player Vector LR {type=number}
 SimlationSpace {type=AnimPhysSimSpaceType}
 SkydiveDiveModeAlpha {type=number}
 SkydiveEnterAlpha {type=number}
 SkydiveFloatingAlpha {type=number}
 SlidingAlpha {type=number}
 SpineSwimmingAlpha {type=number}
 TacticalSprintingAlpha {type=number}
 TargetingAlpha {type=number}
 TurnInTwistAlpha {type=number}
]
